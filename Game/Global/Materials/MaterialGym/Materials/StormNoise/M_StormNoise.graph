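# M_StormNoise is an unlit material in the MaterialGym folder, apparently for a storm noise effect.
# Its graph holds a single call to the MF_StormNoise material function, with Material and Alpha pins, and no other logic.
# The noise itself is defined inside that function, outside this asset.
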Material /Game/Global/Materials/MaterialGym/Materials/StormNoise/M_StormNoise
graph {
 properties {pixelDepthOffsetMode=PDOM_Legacy, shadingModel=MSM_Unlit, usedShadingModels=Unlit}
 nodes [MaterialFunctionCall!(Material, Alpha) {materialFunction=/Game/Global/Materials/MaterialGym/Materials/StormNoise/MaterialFunctions/MF_StormNoise}]
}
